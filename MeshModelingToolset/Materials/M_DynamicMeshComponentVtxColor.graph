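# Two-sided Default Lit material that displays vertex colors on dynamic meshes in the Mesh Modeling Toolset.
# The VertexColorMask parameter tints the vertex color, which is then compensated for exposure with EyeAdaptationInverse and scaled down.
# A world-space normal blends from the vertex normal to a face normal built from screen-space position derivatives, controlled by FlatShading, which defaults to fully flat.
# A note records that a constant meant for Anisotropy causes black jagged edges when used.
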Material /MeshModelingToolset/Materials/M_DynamicMeshComponentVtxColor
graph {
 properties {bTangentSpaceNormal=false, pixelDepthOffsetMode=PDOM_Legacy, twoSided=true, usedShadingModels=Default Lit}
 nodes [
  Constant() {desc="This is supposed to be connected to Anisotropy, but there \r\nis currently a bug creating black jagged edges when it is used.", r=0.1}
  Constant() {r=1}
  Constant() {r=0.8}
  n1 = WorldPosition()
  n2 = DDX(Value=@n1.XYZ)
  n3 = DDY(Value=@n1.XYZ)
  LinearInterpolate(
   A=VertexNormalWS()
   B=Normalize(VectorInput=CrossProduct(A=@n2, B=@n3))
   Alpha=ScalarParameter() {defaultValue=1, parameterName=FlatShading, sliderMax=1}
  )
  n4 = VertexColor()
  n5 = VectorParameter() {parameterName=VertexColorMask, defaultValue=<rgba 1 1 1 1>}
  Multiply(A=EyeAdaptationInverse(LightValueInput=Multiply(A=@n4, B=@n5.RGB), AlphaInput=Constant() {r=1}) {desc=Undo effect of exposure}.EyeAdaptationInverse) {constB=0.1}
 ]
}
